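# Sound cue for the first-person crowd audio of the Swish emote.
# A mixer layers a random crowd loop that fades in with a quieter random one-shot swell that holds briefly, then fades out, both shaped by volume envelopers.
# It uses the first-person emote music sound class and concurrency, plays when silent, and has a send to the music audio-analysis submix.
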SoundCue /BRCosmetics/Sounds/Emotes/Swish/Emote1P_Swish_Crowd_Cue
root=Mixer_0
nodes [
 {
  id=Enveloper_0
  children [Random_0]
  properties {pitchCurve {editorCurveData {keys [{time=0}]}}, volumeCurve {editorCurveData {keys [<at 0 0>,<at 1 1>]}}}
 }
 {
  id=Enveloper_1
  children [Random_1]
  properties {pitchCurve {editorCurveData {keys [{time=0}]}}, volumeCurve {editorCurveData {keys [<at 0 1>,<at 1.5 1>,<at 5 0>]}}}
 }
 {id=Mixer_0, children [Enveloper_0,Enveloper_1], properties {inputVolume [1,0.7]}}
 {id=Random_0, children [WavePlayer_0,WavePlayer_1,WavePlayer_2,WavePlayer_3], properties {weights [1,1,1,1]}}
 {id=Random_1, children [WavePlayer_4,WavePlayer_5,WavePlayer_6,WavePlayer_7], properties {weights [1,1,1,1]}}
 {id=WavePlayer_0, properties {bLooping=true, soundWaveAssetPtr=&/Emote_Swish_Crowd_LP_01}}
 {id=WavePlayer_1, properties {bLooping=true, soundWaveAssetPtr=&/Emote_Swish_Crowd_LP_02}}
 {id=WavePlayer_2, properties {bLooping=true, soundWaveAssetPtr=&/Emote_Swish_Crowd_LP_03}}
 {id=WavePlayer_3, properties {bLooping=true, soundWaveAssetPtr=&/Emote_Swish_Crowd_LP_04}}
 {id=WavePlayer_4, properties {soundWaveAssetPtr=&/Emote_Swish_Crowd_Swell_01}}
 {id=WavePlayer_5, properties {soundWaveAssetPtr=&/Emote_Swish_Crowd_Swell_02}}
 {id=WavePlayer_6, properties {soundWaveAssetPtr=&/Emote_Swish_Crowd_Swell_03}}
 {id=WavePlayer_7, properties {soundWaveAssetPtr=&/Emote_Swish_Crowd_Swell_04}}
]
properties {
 maxDistance=2097152
 soundClassObject=/Game/Sounds/FortSoundClassesAndModes/Fort_Music_Emote_1P_PSM
 virtualizationMode=PlayWhenSilent
 volumeMultiplier=1
 concurrencySet [/Game/Sounds/Concurrency_Rules/Athena_EmoteMusic_1P_Conc]
 soundSubmixSends [{minSendLevel=0, soundSubmix=&Fortnite_AudioAnalysis_Music}]
}
