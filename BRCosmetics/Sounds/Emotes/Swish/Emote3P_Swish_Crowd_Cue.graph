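# Third-person looping crowd sound cue for the Swish emote, classed as third-person emote music.
# A Random node picks one of four equally weighted looping crowd waves and feeds it into an Enveloper that briefly fades the volume in from silence.
# It uses Athena emote attenuation and music concurrency, keeps playing when silent, and sends to the music audio-analysis submix.
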SoundCue /BRCosmetics/Sounds/Emotes/Swish/Emote3P_Swish_Crowd_Cue
root=Enveloper_0
nodes [
 {
  id=Enveloper_0
  children [Random_0]
  properties {pitchCurve {editorCurveData {keys [{time=0}]}}, volumeCurve {editorCurveData {keys [<at 0 0>,<at 0.3 1>]}}}
 }
 {id=Random_0, children [WavePlayer_0,WavePlayer_1,WavePlayer_2,WavePlayer_3], properties {weights [1,1,1,1]}}
 {id=WavePlayer_0, properties {bLooping=true, soundWaveAssetPtr=&/Emote_Swish_Crowd_LP_01}}
 {id=WavePlayer_1, properties {bLooping=true, soundWaveAssetPtr=&/Emote_Swish_Crowd_LP_02}}
 {id=WavePlayer_2, properties {bLooping=true, soundWaveAssetPtr=&/Emote_Swish_Crowd_LP_03}}
 {id=WavePlayer_3, properties {bLooping=true, soundWaveAssetPtr=&/Emote_Swish_Crowd_LP_04}}
]
properties {
 attenuationSettings=&Fort_Emotes_Athena_Att
 maxDistance=1600
 soundClassObject=/Game/Sounds/FortSoundClassesAndModes/Fort_Music_Emote_3P
 virtualizationMode=PlayWhenSilent
 volumeMultiplier=1
 concurrencySet [/Game/Sounds/Concurrency_Rules/Athena_EmoteMusic_Conc]
 soundSubmixSends [{minSendLevel=0, soundSubmix=&Fortnite_AudioAnalysis_Music}]
}
